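# Masked material for the RL105T2 wheel cosmetic, with animated slime, ghost and pulsing "brain" effects.
# It overrides MF_WheelParent_Simple_Alpha attributes, then runs vehicle damage and de-rez transition functions.
# A time-driven cosine pulses vertex offset, while fresnel slime color and parallax-corrected reflections form a diffuse term, both using red vertex color.
# Panning thread normals add to rim normals and distort bump-offset ghost layers for skull- and rim-masked emissive.
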Material /VehicleCosmetics/Wheels/RL105T2/Materials/M_Wheel_RL105T2
graph {
 properties {
  bAutomaticallySetUsageInEditor=false
  bUseMaterialAttributes=true
  bUsedWithMeshDeformer=true
  bUsedWithSkeletalMesh=true
  blendMode=BLEND_Masked
  pixelDepthOffsetMode=PDOM_Legacy
  previewMesh=/VehicleCosmetics/Wheels/Bloom/Meshes/SK_Wheel_Bloom
  usedShadingModels=Default Lit
 }
 nodes [
  n1 = MaterialFunctionCall() {materialFunction=/VehicleCosmetics/Wheels/Mage/Materials/MF_WheelParent_Simple_Alpha}
  n2 = SetMaterialAttributes!(
   MaterialAttributes=@n1.MaterialAttributes
   World Position Offset=NamedRerouteUsage()
   Normal=NamedRerouteUsage()
   Base Color=NamedRerouteUsage()
   Specular=Constant3Vector() {constant=<rgba 0.0743 0.3906 0 1>}
   Emissive Color=NamedRerouteUsage()
   Opacity Mask=TextureSampleParameter2D(UVs=TextureCoordinate() {coordinateIndex=1}) {parameterName=Opsity, texture=/VehicleCosmetics/Textures/T_White}.RGB
  ) {attributeSetTypes [
    #1
    #2
    #3
    #4
    #5
    #6
   ]}
  MaterialFunctionCall(In=MaterialFunctionCall(Material=@n2) {materialFunction=&MF_BaseVehicle_Damage}.Result) {materialFunction=&MF_VehicleDeRez_Transition}
  n3 = Time()
  n4 = ScalarParameter() {defaultValue=0.5, group=SE_Param, parameterName=Brain_Pulse_Speed}
  n5 = TextureCoordinate() {uTiling=2, vTiling=2}
  n6 = Cosine(None=Add(A=Multiply(A=@n3, B=@n4), B=ComponentMask(None=@n5) {g=true}))
  n7 = Constant() {r=1.2}
  n8 = Multiply(
   A=TransformPosition(None=WorldPosition().XYZ) {transformSourceType=TRANSFORMPOSSOURCE_World}
   B=Divide(A=Add(A=@n6, B=@n7), B=Constant() {r=3})
  )
  n9 = ScalarParameter() {defaultValue=0.04, parameterName=Brain_Pulse_Amplitude}
  n10 = VertexColor()
  NamedRerouteDeclaration(None=LinearInterpolate(A=Constant(), B=Multiply(A=@n8, B=@n9), Alpha=AppendVector(A=@n10.R, B=@n10.R))) {name=WPO, nodeColor=<rgba 0 0.1363 1 1>}
  NamedRerouteDeclaration(None=VectorParameter() {group=SE_Param, parameterName=AccentColor, sortPriority=2, defaultValue=<rgba 0.1108 0.2343 1 1>}.RGB) {name=AccCol, nodeColor=<rgba 0.8281 1 0 1>}
  n11 = TextureSampleParameter2D(UVs=NamedRerouteDeclaration(None=TextureCoordinate() {coordinateIndex=1}) {name=TexCoord, nodeColor=<rgba 0 0.5078 1 1>}) {parameterName=NormalSE2, samplerType=SAMPLERTYPE_Normal, texture=/VehicleCosmetics/Wheels/RL105T2/Textures/T_RL105T2_Rim_N}
  n12 = TextureSampleParameter2D(UVs=Panner(Coordinate=TextureCoordinate() {coordinateIndex=1}) {speedX=0.008}) {
   group=SE_Param
   parameterName=Normal_SE
   samplerType=SAMPLERTYPE_Normal
   sortPriority=2
   texture=/VehicleCosmetics/Wheels/RL105T2/Textures/T_RL105T2_Thread_N
  }
  n13 = VectorParameter() {group=SE_Param, parameterName=TireDetailNormalIntensity, sortPriority=2, defaultValue=<rgba 2 2 0 1>}
  n14 = Multiply(A=@n12.RGB, B=@n13.RGB)
  NamedRerouteDeclaration(None=Add(A=@n11.RGB, B=@n14)) {name=add1, nodeColor=<rgba 1 0 0.3516 1>}
  n15 = ScalarParameter() {defaultValue=4, parameterName=SlimeEmiss_power}
  n16 = Fresnel() {baseReflectFraction=0.1, exponent=2}
  n17 = Power(Base=@n16) {constExponent=1}
  n18 = Add(A=Add(A=Multiply(A=@n15, B=@n17) {constA=4}, B=@n17), B=DotProduct(A=NamedRerouteUsage(), B=CameraVectorWS()))
  n19 = Add(A=@n16) {constB=0.01}
  n20 = NamedRerouteUsage()
  n21 = NamedRerouteUsage()
  n22 = TextureSampleParameter2D(UVs=@n21) {parameterName=Emissive Mask, samplerType=SAMPLERTYPE_LinearColor, texture=/VehicleCosmetics/Wheels/RL105T2/Textures/T_RL105T2_Rim_E}
  n23 = LinearInterpolate(
   A=VectorParameter() {group=SE_Param, parameterName=SlimeColor, sortPriority=2, defaultValue=<rgba 0.3621 0.99 0 1>}.RGB
   B=Add(A=@n20, B=@n20)
   Alpha=Add(A=@n22.R, B=@n22.R)
  )
  n24 = Multiply(
   A=Multiply(A=Multiply(A=@n19, B=@n18), B=@n23)
   B=Multiply(
    A=@n23
    B=TextureSampleParameter2D(UVs=@n21) {parameterName=BC mask, samplerType=SAMPLERTYPE_LinearColor, texture=/VehicleCosmetics/Wheels/RL105T2/Textures/T_RL105T2_Rim_BC}.RGB
   )
  )
  n25 = Desaturation(None=@n23, Fraction=Constant() {r=0.2})
  n26 = Constant()
  n27 = Multiply(
   A=Multiply(
    A=MaterialFunctionCall!(Desaturate, Tint=@n26, TintAmount=@n26, ParallaxSize) {materialFunction=/VehicleCosmetics/Wheels/Mild/Materials/MF/MF_ParallexCorrectWorldENV}.WorldSpace
    B=ScalarParameter() {parameterName=WheelReflectionIntensity}
   )
   B=@n25
  )
  n28 = VertexColor()
  NamedRerouteDeclaration(None=LinearInterpolate(B=Add(A=@n24, B=@n27), Alpha=@n28.R)) {name=Diffuse, nodeColor=<rgba 0 0.6719 1 1>}
  n29 = VectorParameter() {group=SE_Param, parameterName=TireDetailNormalIntensity2, sortPriority=2, defaultValue=<rgba 0.5 0.5 0 1>}
  NamedRerouteDeclaration(None=Add(A=Multiply(A=@n14, B=@n29.RGB), B=@n11.RGB)) {name=Normal, nodeColor=<rgba 0.9922 1 0 1>}
  NamedRerouteDeclaration(None=LinearInterpolate(B=NamedRerouteUsage(), Alpha=@n28.R)) {name=WPO2, nodeColor=<rgba 0.6875 0 1 1>}
  NamedRerouteDeclaration(None=@n1.RimRGBAMask) {name=RimM, nodeColor=<rgba 0 1 0.3906 1>}
  n30 = TextureCoordinate() {uTiling=4}
  n31 = Panner(Coordinate=@n30) {speedX=-0.13}
  n32 = ComponentMask(None=Divide(A=@n14) {constB=30}) {g=true, r=true}
  n33 = TextureSample(UVs=BumpOffset(Coordinate=Add(A=@n31, B=@n32), Height=Constant() {r=5})) {samplerType=SAMPLERTYPE_LinearColor, texture=/VehicleCosmetics/Wheels/RL105T2/Textures/T_RL105T2_Ghost_M}
  n34 = Add(A=Panner(Coordinate=@n30) {speedX=-0.35}, B=@n32)
  n35 = Constant() {r=2}
  n36 = BumpOffset(Coordinate=@n31, Height=Constant() {r=-2})
  n37 = Add(
   A=TextureSample(UVs=BumpOffset(Coordinate=@n34, Height=@n35)) {samplerType=SAMPLERTYPE_LinearColor, texture=&T_RL105T2_Ghost_M}.R
   B=Multiply(A=TextureSample(UVs=@n36) {samplerType=SAMPLERTYPE_LinearColor, texture=&T_RL105T2_Ghost_M}.G) {constB=0.5}
  )
  n38 = Power(Base=Multiply(A=@n25, B=Add(A=@n33.G, B=@n37))) {constExponent=1.6}
  n39 = NamedRerouteUsage()
  n40 = TextureSample(UVs=@n39) {samplerType=SAMPLERTYPE_LinearColor, texture=/VehicleCosmetics/Wheels/RL105T2/Textures/T_RL105T2_Skull_M}
  n41 = NamedRerouteUsage()
  n42 = ScalarParameter() {defaultValue=0.5, parameterName=EmissPower}
  n43 = TextureSampleParameter2D(UVs=@n39) {parameterName=Rim_Mask, samplerType=SAMPLERTYPE_LinearColor, texture=/VehicleCosmetics/Wheels/RL105T2/Textures/T_RL105T2_Rim_RGB}
  NamedRerouteDeclaration(None=LinearInterpolate(A=LinearInterpolate(A=@n38, Alpha=@n40.B) {constB=0}, B=Multiply(A=@n41, B=@n42), Alpha=Multiply(A=@n43.B) {constB=0.2})) {name=Emissive, nodeColor=<rgba 1 0 0.8985 1>}
 ]
}
