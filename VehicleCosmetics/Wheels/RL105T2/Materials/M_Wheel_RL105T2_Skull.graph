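# Masked skeletal-mesh material for the skull on the RL105T2 cosmetic vehicle wheel, giving it a pulsing colored glow and animated offset.
# A mask texture on UV1, tinted by CustomColor, drives emissive with brightness modulated by a time-based sine pulse.
# The same sine scales World Position Offset along local Y, so movement stays in sync with the glow.
# The attributes then pass through the shared vehicle damage and de-rez transition functions.
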Material /VehicleCosmetics/Wheels/RL105T2/Materials/M_Wheel_RL105T2_Skull
graph {
 properties {
  bAutomaticallySetUsageInEditor=false
  bUseMaterialAttributes=true
  bUsedWithSkeletalMesh=true
  blendMode=BLEND_Masked
  usedShadingModels=Default Lit
 }
 nodes [
  n1 = TextureCoordinate() {coordinateIndex=1}
  n2 = TextureSampleParameter2D(UVs=@n1) {parameterName=Mask, samplerType=SAMPLERTYPE_LinearColor, texture=/VehicleCosmetics/Wheels/RL105T2/Textures/T_RL105T2_Skull_M}
  n3 = Add(A=@n2.R, B=@n2.R)
  n4 = VectorParameter() {group=SE_Param, parameterName=CustomColor, sortPriority=2, defaultValue=<rgba 0.0685 0.0578 0.8469 1>}
  n5 = Sine(None=Multiply(A=Time(), B=ScalarParameter() {defaultValue=0.63, parameterName=PulseSpeed}))
  n6 = Add(B=@n5) {constA=1}
  n7 = Clamp(None=Power(Base=Divide(A=@n6) {constB=1.75}) {constExponent=3})
  n8 = Multiply(A=Multiply(A=@n3, B=@n4.RGB), B=Multiply(A=ScalarParameter() {defaultValue=2, parameterName=EmissPower}, B=Add(A=@n3, B=@n7)))
  n9 = TextureSampleParameter2D(UVs=@n1) {parameterName=Opasity, texture=/VehicleCosmetics/Textures/T_White}
  n10 = Multiply(
   A=Multiply(A=Add(A=@n5) {constB=1.5}, B=ScalarParameter() {defaultValue=2, parameterName=WPOstrength})
   B=Transform(None=Constant3Vector() {constant=<rgba 0 1 0 0>}) {transformSourceType=TRANSFORMSOURCE_Local}
  )
  MaterialFunctionCall(In=MaterialFunctionCall(Material=SetMaterialAttributes!(MaterialAttributes, Emissive Color=@n8, Opacity=@n9.RGB, World Position Offset=@n10) {attributeSetTypes [#1, #2, #3]}) {materialFunction=&MF_BaseVehicle_Damage}.Result) {materialFunction=&MF_VehicleDeRez_Transition}
 ]
}
