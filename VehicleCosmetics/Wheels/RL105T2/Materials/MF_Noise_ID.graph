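# Material function that turns a scalar input into a procedural 1D noise wave, previewed with scaled Time.
# The input's integer part seeds a sine-based pseudo-random coefficient for each unit interval.
# Its fractional part shapes a quartic curve that is zero at interval ends, with a sign flip driven by the divided input's fraction.
# This resembles Inigo Quilez's cheap polynomial 1D noise, needing no texture lookup.
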MaterialFunction /VehicleCosmetics/Wheels/RL105T2/Materials/MF_Noise_ID
graph {nodes [
  n1 = FunctionInput(Preview=Multiply(A=Time()) {constB=3}) {inputType=FunctionInput_Scalar}
  n2 = Frac(None=@n1)
  n3 = Sine(None=Floor(None=@n1))
  n4 = Multiply(A=Add(A=Multiply(A=@n3) {constB=0.5}) {constB=0.5}) {constB=16}
  n5 = Multiply(A=Subtract(A=@n2), B=Multiply(A=Subtract(A=@n4) {constB=4}, B=@n2))
  n6 = Subtract(A=@n2)
  n7 = Frac(None=Divide(A=@n1))
  n8 = Constant()
  n9 = Multiply(A=@n2, B=If(A=Subtract(A=@n7) {constB=0.5}, B=@n8, "A > B"=Constant() {r=1}, "A == B"=@n8, "A < B"=Constant() {r=-1}))
  FunctionOutput(None=Multiply(A=Subtract(A=@n5), B=Multiply(A=@n6, B=@n9)))
 ]}
