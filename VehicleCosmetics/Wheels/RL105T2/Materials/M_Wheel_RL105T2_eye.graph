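# Lit skeletal-mesh material for the eye on the RL105T2 cosmetic vehicle wheel, animating an erratically moving eye.
# Clamped, time-driven noise replaces Panner time, shifting U and V separately for color and normal textures.
# Masked color gains CustomColor through texture alpha, plus a masked world-environment term scaled by Fresnel.
# The attributes pass through shared vehicle damage and de-rez transition functions.
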Material /VehicleCosmetics/Wheels/RL105T2/Materials/M_Wheel_RL105T2_eye
graph {
 properties {bAutomaticallySetUsageInEditor=false, bUseMaterialAttributes=true, bUsedWithSkeletalMesh=true, usedShadingModels=Default Lit}
 nodes [
  n1 = TextureCoordinate()
  n2 = TextureSample(UVs=@n1) {samplerType=SAMPLERTYPE_LinearColor, texture=/VehicleCosmetics/Wheels/RL105T2/Textures/T_RL105T2_Eye_M}
  n3 = Time()
  n4 = Multiply(B=MaterialFunctionCall!(In=Multiply(A=@n3) {constB=0.5}) {materialFunction=&/MF_Noise_ID}.Result) {constA=30}
  n5 = Subtract() {constA=0, constB=2}
  n6 = AppendVector(A=@n5, B=@n5)
  n7 = Time()
  n8 = Multiply(A=@n7) {constB=1.5}
  n9 = Clamp(None=LinearInterpolate(
    A=AppendVector(A=@n6, B=@n5)
    B=Add() {constA=2}
    Alpha=MaterialFunctionCall(In=@n8) {materialFunction=&MF_Noise_ID}.Result
   )) {minDefault=-1}
  n10 = Add(A=Multiply(B=Clamp(None=@n4) {minDefault=-1}) {constA=2.3}, B=Multiply(B=ComponentMask(None=@n9) {r=true}) {constA=0.4})
  n11 = Multiply(B=MaterialFunctionCall(In=Multiply(B=@n3) {constA=-0.5}) {materialFunction=&MF_Noise_ID}.Result) {constA=30}
  n12 = Subtract() {constA=0, constB=2}
  n13 = AppendVector(A=@n12, B=@n12)
  n14 = Multiply(B=@n7) {constA=1}
  n15 = Clamp(None=LinearInterpolate(
    A=AppendVector(A=@n13, B=@n12)
    B=Add() {constA=2}
    Alpha=MaterialFunctionCall(In=@n14) {materialFunction=&MF_Noise_ID}.Result
   )) {minDefault=-1}
  n16 = Add(A=Multiply(A=Clamp(None=@n11) {minDefault=-1}) {constB=2.3}, B=Multiply(B=ComponentMask(None=@n15) {r=true}) {constA=0.4})
  n17 = AppendVector(
   A=ComponentMask(None=Panner(Coordinate=@n1, Time=@n10) {speedX=0.04, speedY=0.02}) {r=true}
   B=ComponentMask(None=Panner(Coordinate=@n1, Time=@n16) {speedY=0.04}) {g=true}
  )
  n18 = TextureSample(UVs=@n17) {texture=/VehicleCosmetics/Wheels/RL105T2/Textures/T_RL105T2_Eye_BC}
  n19 = Multiply(A=@n2.RGB, B=@n18.RGB)
  n20 = Multiply(A=VectorParameter() {group=SE_Param, parameterName=CustomColor, sortPriority=2, defaultValue=<rgba 0 0.0124 0.99 1>}.RGB, B=@n18.A)
  n21 = MaterialFunctionCall!(
   Tint=ScalarParameter() {defaultValue=1, parameterName=WorldENV Tint Amount}
   TintAmount=ScalarParameter() {defaultValue=1, parameterName=WorldENV Desaturate}
   Desaturate=VectorParameter() {parameterName=WorldENV Tint Color, defaultValue=<rgba 0 0 0 1>}.RGB
  ) {materialFunction=/VehicleCosmetics/Wheels/Oldmans/New/Materials/MF_WordENV}
  n22 = MaterialFunctionCall!(Input Normal, Input Sharpness, Input Two Sided) {materialFunction=/VehicleCosmetics/Wheels/CarCar_Holographic/Materials/MF_Fresnel_Advanced}
  n23 = Add(
   A=LinearInterpolate(A=@n19, B=Add(A=@n20, B=@n19), Alpha=@n18.A)
   B=Multiply(A=Multiply(A=@n21.WorldSpace_1, B=@n2.R), B=Add(A=@n22.Edge) {constB=0.5})
  )
  n24 = TextureSample(UVs=@n17) {samplerType=SAMPLERTYPE_Normal, texture=/VehicleCosmetics/Wheels/RL105T2/Textures/T_RL105T2_Eye_N}
  MaterialFunctionCall(In=MaterialFunctionCall(Material=SetMaterialAttributes!(MaterialAttributes, Base Color=@n23, Normal=@n24.RGB) {attributeSetTypes [#1, #2]}) {materialFunction=&MF_BaseVehicle_Damage}.Result) {materialFunction=&MF_VehicleDeRez_Transition}
 ]
}
